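# Level Sequence for the CanyonVale_M skin's lobby idle animation, authored at 30 fps on a spawned FortPlayerMannequin with Head and Body mesh bindings.
# A Fortnite_3L_OnFace_CtrlRig track holds a mostly constant mouth and brow pose, with keyed blinks, eye darts and one brief jaw and left-brow twitch.
# An MR_Biped_v02 body rig track is present without listed channels.
# Animated face channels return to their starting values by frame 341, allowing a seamless loop.
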LevelSequence /BRCosmetics/Animation/Game/MainPlayer/Menu/BR/CustomSkins/_S41/CanyonVale_M/CanyonVale_Lobby_idle_LS
bindings [
 Head {id=#1, parent_id=#2, possessed_class=SkeletalMeshComponentBudgeted, tracks []}
 mannequin_cranyonVale {
  id=#2
  possessed_class=FortPlayerMannequin
  tracks [
   {type=MovieSceneSpawnTrack, sections [{type=MovieSceneSpawnSection, channels [None {type=MovieSceneScriptingBoolChannel, default=true}]}]}
   {
    type=MovieScene3DTransformTrack
    sections [{
      type=MovieScene3DTransformSection
      channels [
       Location.X {type=MovieSceneScriptingDoubleChannel, default=0}
       Location.Y {type=MovieSceneScriptingDoubleChannel, default=0}
       Location.Z {type=MovieSceneScriptingDoubleChannel, default=0}
       Rotation.X {type=MovieSceneScriptingDoubleChannel, default=0}
       Rotation.Y {type=MovieSceneScriptingDoubleChannel, default=0}
       Rotation.Z {type=MovieSceneScriptingDoubleChannel, default=0}
       Scale.X {type=MovieSceneScriptingDoubleChannel, default=1}
       Scale.Y {type=MovieSceneScriptingDoubleChannel, default=1}
       Scale.Z {type=MovieSceneScriptingDoubleChannel, default=1}
      ]
     }]
   }
   {type=MovieSceneControlRigParameterTrack, display_name=MR_Biped_v02, sections [{type=MovieSceneControlRigParameterSection}]}
   {
    type=MovieSceneControlRigParameterTrack
    display_name=Fortnite_3L_OnFace_CtrlRig
    sections [{
      type=MovieSceneControlRigParameterSection
      channels [
       CTRL_C_eyesAim.Location.X {type=MovieSceneScriptingFloatChannel, default=0, const [0,12]}
       CTRL_C_eyesAim.Location.Y {type=MovieSceneScriptingFloatChannel, default=0, const [0,12]}
       CTRL_C_eyesAim.Location.Z {type=MovieSceneScriptingFloatChannel, default=0, const [0,12]}
       CTRL_C_eyesAim.Rotation.X {type=MovieSceneScriptingFloatChannel, default=0, const [0,12]}
       CTRL_C_eyesAim.Rotation.Y {type=MovieSceneScriptingFloatChannel, default=0, const [0,12]}
       CTRL_C_eyesAim.Rotation.Z {type=MovieSceneScriptingFloatChannel, default=0, const [0,12]}
       CTRL_C_eyesAim.Scale.X {type=MovieSceneScriptingFloatChannel, default=1, const [1,12]}
       CTRL_C_eyesAim.Scale.Y {type=MovieSceneScriptingFloatChannel, default=1, const [1,12]}
       CTRL_C_eyesAim.Scale.Z {type=MovieSceneScriptingFloatChannel, default=1, const [1,12]}
       CTRL_L_eyeAim.Location.X {type=MovieSceneScriptingFloatChannel, default=0, const [0,12]}
       CTRL_L_eyeAim.Location.Y {type=MovieSceneScriptingFloatChannel, default=0, const [0,12]}
       CTRL_L_eyeAim.Location.Z {type=MovieSceneScriptingFloatChannel, default=0, const [0,12]}
       CTRL_L_eyeAim.Rotation.X {type=MovieSceneScriptingFloatChannel, default=0, const [0,12]}
       CTRL_L_eyeAim.Rotation.Y {type=MovieSceneScriptingFloatChannel, default=0, const [0,12]}
       CTRL_L_eyeAim.Rotation.Z {type=MovieSceneScriptingFloatChannel, default=0, const [0,12]}
       CTRL_L_eyeAim.Scale.X {type=MovieSceneScriptingFloatChannel, default=1, const [1,12]}
       CTRL_L_eyeAim.Scale.Y {type=MovieSceneScriptingFloatChannel, default=1, const [1,12]}
       CTRL_L_eyeAim.Scale.Z {type=MovieSceneScriptingFloatChannel, default=1, const [1,12]}
       CTRL_R_eyeAim.Location.X {type=MovieSceneScriptingFloatChannel, default=0, const [0,12]}
       CTRL_R_eyeAim.Location.Y {type=MovieSceneScriptingFloatChannel, default=0, const [0,12]}
       CTRL_R_eyeAim.Location.Z {type=MovieSceneScriptingFloatChannel, default=0, const [0,12]}
       CTRL_R_eyeAim.Rotation.X {type=MovieSceneScriptingFloatChannel, default=0, const [0,12]}
       CTRL_R_eyeAim.Rotation.Y {type=MovieSceneScriptingFloatChannel, default=0, const [0,12]}
       CTRL_R_eyeAim.Rotation.Z {type=MovieSceneScriptingFloatChannel, default=0, const [0,12]}
       CTRL_R_eyeAim.Scale.X {type=MovieSceneScriptingFloatChannel, default=1, const [1,12]}
       CTRL_R_eyeAim.Scale.Y {type=MovieSceneScriptingFloatChannel, default=1, const [1,12]}
       CTRL_R_eyeAim.Scale.Z {type=MovieSceneScriptingFloatChannel, default=1, const [1,12]}
       CTRL_convergenceSwitch {type=MovieSceneScriptingFloatChannel, default=0, const [0,12]}
       follow_head {type=MovieSceneScriptingFloatChannel, default=0, const [0,12]}
       Face_Ctrl_Master.Location.X {type=MovieSceneScriptingFloatChannel, default=0, const [0,12]}
       Face_Ctrl_Master.Location.Y {type=MovieSceneScriptingFloatChannel, default=0, const [0,12]}
       Face_Ctrl_Master.Location.Z {type=MovieSceneScriptingFloatChannel, default=0, const [0,12]}
       Face_Ctrl_Master.Rotation.X {type=MovieSceneScriptingFloatChannel, default=0, const [0,12]}
       Face_Ctrl_Master.Rotation.Y {type=MovieSceneScriptingFloatChannel, default=0, const [0,12]}
       Face_Ctrl_Master.Rotation.Z {type=MovieSceneScriptingFloatChannel, default=0, const [0,12]}
       Face_Ctrl_Master.Scale.X {type=MovieSceneScriptingFloatChannel, default=1, const [1,12]}
       Face_Ctrl_Master.Scale.Y {type=MovieSceneScriptingFloatChannel, default=1, const [1,12]}
       Face_Ctrl_Master.Scale.Z {type=MovieSceneScriptingFloatChannel, default=1, const [1,12]}
       gui_follow_head {type=MovieSceneScriptingFloatChannel, default=1, const [1,12]}
       CTRL_L_eye_blink {
        type=MovieSceneScriptingFloatChannel
        default=0
        thinned_from=28
        keys [[0,0],[35,0],[42,0.7302],[50,-0.2319],[57,0],[76,0],[80,0.6097],[83,0],[109,0],[115,0.5941],[118,0.802],[123,0],[130,-0.6941],[137,0],[225,0],[230,0.057],[235,1],[239,0],[269,0],[272,0.9525],[281,0],[341,0]]
       }
       CTRL_L_eye_eyelidU.X {type=MovieSceneScriptingFloatChannel, default=0, thinned_from=24, keys [[0,0],[341,0]]}
       CTRL_L_eye_eyelidU.Y {type=MovieSceneScriptingFloatChannel, default=0, thinned_from=24, keys [[0,-0.2642],[341,-0.2641]]}
       CTRL_L_eye_eyelidD.X {type=MovieSceneScriptingFloatChannel, default=0, const [0,24]}
       CTRL_L_eye_eyelidD.Y {type=MovieSceneScriptingFloatChannel, default=0, const [0,24]}
       CTRL_L_brow.X {type=MovieSceneScriptingFloatChannel, default=0, const [0,24]}
       CTRL_L_brow.Y {type=MovieSceneScriptingFloatChannel, default=0, const [-0.9342,24]}
       CTRL_L_brow_inner {type=MovieSceneScriptingFloatChannel, default=0, const [0,24]}
       CTRL_L_brow_outer {type=MovieSceneScriptingFloatChannel, default=0, thinned_from=24, keys [[0,0],[123,0],[130,1],[137,0],[341,0]]}
       CTRL_C_nose.X {type=MovieSceneScriptingFloatChannel, default=0, thinned_from=12, keys [[0,0],[341,0]]}
       CTRL_C_nose.Y {type=MovieSceneScriptingFloatChannel, default=0, thinned_from=12, keys [[0,0.0379],[341,0.0378]]}
       CTRL_L_nose.X {type=MovieSceneScriptingFloatChannel, default=0, const [-0.9183,12]}
       CTRL_L_nose.Y {type=MovieSceneScriptingFloatChannel, default=0, const [1,12]}
       noseWrinkleUpper {type=MovieSceneScriptingFloatChannel, default=0, const [0,12]}
       CTRL_R_nose.X {type=MovieSceneScriptingFloatChannel, default=0, const [-0.9183,12]}
       CTRL_R_nose.Y {type=MovieSceneScriptingFloatChannel, default=0, const [1,12]}
       noseWrinkleUpper_2 {type=MovieSceneScriptingFloatChannel, default=0, const [0,12]}
       CTRL_L_eye_squintInner {type=MovieSceneScriptingFloatChannel, default=0, const [0,24]}
       CTRL_L_eye_cheekRaise {type=MovieSceneScriptingFloatChannel, default=0, const [0,12]}
       CTRL_C_mouth.X {type=MovieSceneScriptingFloatChannel, default=0, thinned_from=13, keys [[0,0],[341,0]]}
       CTRL_C_mouth.Y {type=MovieSceneScriptingFloatChannel, default=0, thinned_from=13, keys [[0,0.0813],[341,0.0813]]}
       CTRL_L_mouth_lipU.X {type=MovieSceneScriptingFloatChannel, default=0, const [0,12]}
       CTRL_L_mouth_lipU.Y {type=MovieSceneScriptingFloatChannel, default=0, const [-1,12]}
       mouthPress {type=MovieSceneScriptingFloatChannel, default=0, const [0,12]}
       lipsTogether {type=MovieSceneScriptingFloatChannel, default=0, const [0,12]}
       CTRL_L_mouth_lipD.X {type=MovieSceneScriptingFloatChannel, default=0, thinned_from=12, keys [[0,0],[341,0]]}
       CTRL_L_mouth_lipD.Y {type=MovieSceneScriptingFloatChannel, default=0, const [-0.8108,12]}
       mouthPress_2 {type=MovieSceneScriptingFloatChannel, default=0, const [0,12]}
       lipsTogether_2 {type=MovieSceneScriptingFloatChannel, default=0, const [0,12]}
       CTRL_L_mouth_smileDepress.X {type=MovieSceneScriptingFloatChannel, default=0, const [0.8229,12]}
       CTRL_L_mouth_smileDepress.Y {type=MovieSceneScriptingFloatChannel, default=0, const [0,12]}
       CTRL_L_mouth_corner.X {type=MovieSceneScriptingFloatChannel, default=0, const [-0.6549,12]}
       CTRL_L_mouth_corner.Y {type=MovieSceneScriptingFloatChannel, default=0, const [-1,12]}
       lipsBlow {type=MovieSceneScriptingFloatChannel, default=0, const [0,12]}
       mouthDimple {type=MovieSceneScriptingFloatChannel, default=0, const [0,12]}
       mouthStretch {type=MovieSceneScriptingFloatChannel, default=0, const [0,12]}
       mouthPurseU {type=MovieSceneScriptingFloatChannel, default=0, const [0,12]}
       mouthPurseD {type=MovieSceneScriptingFloatChannel, default=0, const [0,12]}
       mouthFunnelU {type=MovieSceneScriptingFloatChannel, default=0, const [0,12]}
       mouthFunnelD {type=MovieSceneScriptingFloatChannel, default=0, const [0,12]}
       mouthCornerSharpnessU {type=MovieSceneScriptingFloatChannel, default=0, const [0,12]}
       mouthCornerSharpnessD {type=MovieSceneScriptingFloatChannel, default=0, const [0,12]}
       CTRL_R_mouth_lipU.X {type=MovieSceneScriptingFloatChannel, default=0, const [0,12]}
       CTRL_R_mouth_lipU.Y {type=MovieSceneScriptingFloatChannel, default=0, const [-1,12]}
       mouthPress_3 {type=MovieSceneScriptingFloatChannel, default=0, const [0,12]}
       lipsTogether_3 {type=MovieSceneScriptingFloatChannel, default=0, const [0,12]}
       CTRL_R_mouth_lipD.X {type=MovieSceneScriptingFloatChannel, default=0, thinned_from=12, keys [[0,0],[341,0]]}
       CTRL_R_mouth_lipD.Y {type=MovieSceneScriptingFloatChannel, default=0, const [-0.8108,12]}
       mouthPress_2_2 {type=MovieSceneScriptingFloatChannel, default=0, const [0,12]}
       lipsTogether_2_2 {type=MovieSceneScriptingFloatChannel, default=0, const [0,12]}
       CTRL_R_mouth_smileDepress.X {type=MovieSceneScriptingFloatChannel, default=0, const [0.8797,12]}
       CTRL_R_mouth_smileDepress.Y {type=MovieSceneScriptingFloatChannel, default=0, const [0,12]}
       CTRL_R_mouth_corner.X {type=MovieSceneScriptingFloatChannel, default=0, const [-0.4862,12]}
       CTRL_R_mouth_corner.Y {type=MovieSceneScriptingFloatChannel, default=0, const [-0.3573,12]}
       lipsBlow_2 {type=MovieSceneScriptingFloatChannel, default=0, const [0,12]}
       mouthDimple_2 {type=MovieSceneScriptingFloatChannel, default=0, const [0,12]}
       mouthStretch_2 {type=MovieSceneScriptingFloatChannel, default=0, const [0,12]}
       mouthPurseU_2 {type=MovieSceneScriptingFloatChannel, default=0, const [0,12]}
       mouthPurseD_2 {type=MovieSceneScriptingFloatChannel, default=0, const [0,12]}
       mouthFunnelU_2 {type=MovieSceneScriptingFloatChannel, default=0, const [0,12]}
       mouthFunnelD_2 {type=MovieSceneScriptingFloatChannel, default=0, const [0,12]}
       mouthCornerSharpnessU_2 {type=MovieSceneScriptingFloatChannel, default=0, const [0,12]}
       mouthCornerSharpnessD_2 {type=MovieSceneScriptingFloatChannel, default=0, const [0,12]}
       CTRL_L_mouth_suckBlow {type=MovieSceneScriptingFloatChannel, default=0, const [-0.8995,12]}
       CTRL_C_jaw.Location.X {
        type=MovieSceneScriptingFloatChannel
        default=0
        thinned_from=15
        keys [[0,0],[115,0],[117,0.1219],[118,0.6983],[124,0.6983],[126,-0.342],[130,0],[341,0]]
       }
       CTRL_C_jaw.Location.Y {
        type=MovieSceneScriptingFloatChannel
        default=0
        thinned_from=15
        keys [[0,-0.1509],[115,-0.1509],[117,0.0969],[118,0.099],[123,0.0933],[124,0.1394],[126,0.1621],[130,0.1138],[137,-0.1509],[341,-0.1509]]
       }
       CTRL_C_jaw.Location.Z {type=MovieSceneScriptingFloatChannel, default=0, const [0.0028,15]}
       jawOpenExtreme {type=MovieSceneScriptingFloatChannel, default=0, const [0,13]}
       chinRaiseUL {type=MovieSceneScriptingFloatChannel, default=0, const [0,13]}
       chinRaiseUR {type=MovieSceneScriptingFloatChannel, default=0, const [0,13]}
       chinRaiseDL {type=MovieSceneScriptingFloatChannel, default=0, const [0,13]}
       chinRaiseDR {type=MovieSceneScriptingFloatChannel, default=0, const [0,13]}
       CTRL_R_brow.X {type=MovieSceneScriptingFloatChannel, default=0, const [0,24]}
       CTRL_R_brow.Y {type=MovieSceneScriptingFloatChannel, default=0, const [-0.9137,24]}
       CTRL_R_brow_inner {type=MovieSceneScriptingFloatChannel, default=0, const [0,23]}
       CTRL_R_brow_outer {type=MovieSceneScriptingFloatChannel, default=0, const [0,24]}
       CTRL_R_eye_blink {
        type=MovieSceneScriptingFloatChannel
        default=0
        thinned_from=28
        keys [[0,0],[35,0],[42,0.7302],[50,-0.2319],[57,0],[76,0],[80,0.6172],[83,0],[109,0],[115,0.7407],[118,1],[123,0],[130,-0.4766],[137,0],[225,0],[230,0.057],[235,1],[239,0],[269,0],[272,0.9525],[281,0],[341,0]]
       }
       CTRL_R_eye_eyelidU.X {type=MovieSceneScriptingFloatChannel, default=0, const [0,24]}
       CTRL_R_eye_eyelidU.Y {type=MovieSceneScriptingFloatChannel, default=0, const [0,24]}
       CTRL_R_eye_eyelidD.X {type=MovieSceneScriptingFloatChannel, default=0, const [0,24]}
       CTRL_R_eye_eyelidD.Y {type=MovieSceneScriptingFloatChannel, default=0, const [0,24]}
       CTRL_R_eye_squintInner {type=MovieSceneScriptingFloatChannel, default=0, const [0,24]}
       CTRL_R_eye_cheekRaise {type=MovieSceneScriptingFloatChannel, default=0, const [0,12]}
       CTRL_R_mouth_suckBlow {type=MovieSceneScriptingFloatChannel, default=0, const [-0.9654,12]}
       CTRL_C_eye.X {
        type=MovieSceneScriptingFloatChannel
        default=0
        thinned_from=27
        keys [[0,0.05],[44,0.05],[48,-0.7413],[78,-0.7413],[83,0.6233],[92,0.6391],[103,0.4081],[109,0.4716],[118,0.4716],[123,0.0693],[235,0.0693],[240,0.5888],[263,0.9283],[272,0.9283],[276,-0.3756],[290,-0.1152],[297,0.1166],[326,0.0679],[341,0.05]]
       }
       CTRL_C_eye.Y {
        type=MovieSceneScriptingFloatChannel
        default=0
        thinned_from=27
        keys [[0,0.015],[44,0.015],[46,-0.3538],[48,0.015],[78,0.015],[83,-0.0012],[92,-0.0025],[103,-0.0035],[109,-0.0037],[145,-0.0037],[174,0.2484],[225,-0.0037],[272,-0.0037],[276,0.1046],[290,0.1046],[297,0.012],[326,0.0142],[341,0.015]]
       }
       CTRL_L_eye.X {
        type=MovieSceneScriptingFloatChannel
        default=0
        thinned_from=25
        keys [[0,0],[92,0],[103,-0.1114],[109,0],[118,0],[123,-0.4024],[130,0],[239,0],[240,-0.3084],[245,0],[341,0]]
       }
       CTRL_L_eye.Y {type=MovieSceneScriptingFloatChannel, default=0, thinned_from=25, keys [[0,0],[341,0]]}
       CTRL_R_eye.X {type=MovieSceneScriptingFloatChannel, default=0, const [0,24]}
       CTRL_R_eye.Y {type=MovieSceneScriptingFloatChannel, default=0, const [0,24]}
       eyesConvergence {type=MovieSceneScriptingFloatChannel, default=0, const [1,1]}
       eyelidsReduce {type=MovieSceneScriptingFloatChannel, default=0, const [0,22]}
       lookAtSwitch {type=MovieSceneScriptingFloatChannel, default=0, const [0,12]}
       CTRL_C_tongue.Location.X {type=MovieSceneScriptingFloatChannel, default=0, const [-0.5794,12]}
       CTRL_C_tongue.Location.Y {type=MovieSceneScriptingFloatChannel, default=0, const [0,12]}
       CTRL_C_tongue.Location.Z {type=MovieSceneScriptingFloatChannel, default=0, const [0.0994,12]}
       wideNarrow {type=MovieSceneScriptingFloatChannel, default=0, const [0,12]}
       rollUpDown {type=MovieSceneScriptingFloatChannel, default=0, const [0,12]}
       rollLeftRight {type=MovieSceneScriptingFloatChannel, default=0, const [0,12]}
       CTRL_C_teethU.Location.X {type=MovieSceneScriptingFloatChannel, default=0, const [0,12]}
       CTRL_C_teethU.Location.Y {type=MovieSceneScriptingFloatChannel, default=0, const [0,12]}
       CTRL_C_teethU.Location.Z {type=MovieSceneScriptingFloatChannel, default=0, const [0,12]}
       rotateFwdBack {type=MovieSceneScriptingFloatChannel, default=0, const [0,12]}
       CTRL_C_teethD.Location.X {type=MovieSceneScriptingFloatChannel, default=0, const [0,12]}
       CTRL_C_teethD.Location.Y {type=MovieSceneScriptingFloatChannel, default=0, const [0,12]}
       CTRL_C_teethD.Location.Z {type=MovieSceneScriptingFloatChannel, default=0, const [0,12]}
       rotateFwdBack_2 {type=MovieSceneScriptingFloatChannel, default=0, const [0,12]}
      ]
     }]
   }
  ]
 }
 Body {id=#3, parent_id=#2, possessed_class=SkeletalMeshComponentBudgeted, tracks []}
]
display_rate {denominator=1, numerator=30}
tick_resolution {denominator=1, numerator=24000}
tracks []
